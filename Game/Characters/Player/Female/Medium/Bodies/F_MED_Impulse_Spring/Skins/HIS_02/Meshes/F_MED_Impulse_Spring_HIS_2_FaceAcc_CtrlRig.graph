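# Face accessory Control Rig for the F_MED_Impulse_Spring HIS_02 player skin that adjusts hair bones as the head turns.
# A spherical pose reader measures the head bone relative to spine_05 within an offset cone with falloff, with debug drawing on.
# Its output weights three chained ModifyTransforms nodes on hair base bones.
# The right B strand shifts and narrows and the back strand shifts, while the left B strand's offset is identity.
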ControlRigBlueprint /Game/Characters/Player/Female/Medium/Bodies/F_MED_Impulse_Spring/Skins/HIS_02/Meshes/F_MED_Impulse_Spring_HIS_2_FaceAcc_CtrlRig
graphs [
 RigVMFunctionLibrary {links [], nodes []}
 RigVMModel {nodes [
   RigUnit_BeginExecution = &RigUnit_BeginExecution()
   n1 = /Script/ControlRig.RigUnit_SequenceExecution(ExecuteContext=@RigUnit_BeginExecution.ExecutePin) {is=struct RigVMUnitNode}
   n2 = &RigUnit_SphericalPoseReader(
    ActiveRegionScaleFactors=(PositiveWidth=1,NegativeWidth=1,PositiveHeight=1,NegativeHeight=1)
    ActiveRegionSize=0.05
    Debug=(bDrawDebug=True,DebugScale=25,DebugSegments=20,DebugThickness=0.25)
    DriverAxis=(X=1,Y=0,Z=0)
    DriverItem=(Type=Bone,Name="head")
    FalloffRegionScaleFactors=(PositiveWidth=1,NegativeWidth=1,PositiveHeight=1,NegativeHeight=1)
    FalloffSize=0.15
    OptionalParentItem=(Type=Bone,Name="spine_05")
    RotationOffset=(X=-40,Y=-30,Z=20)
    ExecutePin=@n1.A
   )
   > &RigUnit_ModifyTransforms(
    ItemToModify=((Item=(Type=Bone),Transform=(Rotation=(X=0,Y=0,Z=0,W=1),Translation=(X=0,Y=0,Z=0),Scale3D=(X=1,Y=1,Z=1))),(Item=(Type=Bone,Name="his1_hair_B_a_base_r"),Transform=(Rotation=(X=0,Y=0,Z=0,W=1),Translation=(X=-7,Y=0,Z=0),Scale3D=(X=0.8,Y=1,Z=1))))
    WeightMinimum=0
    Weight=@n2.OutputParam
   )
   > &RigUnit_ModifyTransforms(
    ItemToModify=((Item=(Type=Bone),Transform=(Rotation=(X=0,Y=0,Z=0,W=1),Translation=(X=0,Y=0,Z=0),Scale3D=(X=1,Y=1,Z=1))),(Item=(Type=Bone,Name="his1_hair_B_a_base_l"),Transform=(Rotation=(X=0,Y=0,Z=0,W=1),Translation=(X=0,Y=0,Z=0),Scale3D=(X=1,Y=1,Z=1))))
    WeightMinimum=0
    Weight=@n2.OutputParam
   )
   > &RigUnit_ModifyTransforms(
    ItemToModify=((Item=(Type=Bone),Transform=(Rotation=(X=0,Y=0,Z=0,W=1),Translation=(X=0,Y=0,Z=0),Scale3D=(X=1,Y=1,Z=1))),(Item=(Type=Bone,Name="his1_hair_bk_a_base_l"),Transform=(Rotation=(X=0,Y=0,Z=0,W=1),Translation=(X=-2,Y=0,Z=0),Scale3D=(X=1,Y=1,Z=1))))
    WeightMinimum=0
    Weight=@n2.OutputParam
   )
  ]}
]
hierarchy [
 root {loc=<0 0 0>, rot=<0 0 0 1>}
 pelvis {parent=root, loc=<0 -0.5301 98.0571>, rot=<0.0127 0.707 -0.0127 -0.707>}
 spine_01 {parent=pelvis, loc=<6.1208 -0.9328 0>, rot=<0 0 0.034 -0.9994>}
 spine_02 {parent=spine_01, loc=<6.5996 -0.2095 0>, rot=<0 0 0.0238 -0.9997>}
 spine_03 {parent=spine_02, loc=<6.4914 -0.1237 0>, rot=<0 0 0.0149 -0.9999>}
 spine_04 {parent=spine_03, loc=<6.9598 0.0794 0>, rot=<0 0 -0.0132 -0.9999>}
 spine_05 {parent=spine_04, loc=<6.9987 0 0>, rot=<0 0 -0.0015 -1>}
 clavicle_l {parent=spine_05, loc=<5.4831 -0.1053 -3.2946>, rot=<0.0465 -0.7289 -0.009 -0.6829>}
 upperarm_l {parent=clavicle_l, loc=<9.8658 -1.2136 -0.4105>, rot=<0.0164 -0.3245 -0.0269 -0.9454>}
 neck_01 {parent=spine_05, loc=<11.3138 0.1529 0>, rot=<0 0 -0.2003 -0.9797>}
 neck_02 {parent=neck_01, loc=<4.5627 0 0>, rot=<0 0 0.0504 -0.9987>}
 head {parent=neck_02, loc=<5.3914 0 0>, rot=<0 0 0.115 -0.9934>}
 his1_hair {parent=head, loc=<0.3282 -0.7471 0>, rot=<0 0 0 -1>}
 his1_hair_bk_a_base_l {parent=his1_hair, loc=<8.463 -6.5457 0.0182>, rot=<-0.0031 -1 -0.0091 -0.0011>}
 his1_hair_bk_a {parent=his1_hair_bk_a_base_l, loc=<0 0 0>, rot=<0 0 0 1>}
 his1_hair_bk_b {parent=his1_hair_bk_a, loc=<5.0729 -0.3026 0>, rot=<0 0 0 1>}
 his1_hair_B_a_base_l {parent=his1_hair, loc=<7.6055 -3.5797 -5.7833>, rot=<0.0321 -0.8842 -0.4616 0.0632>}
 his1_hair_B_a_l {parent=his1_hair_B_a_base_l, loc=<0 0 0>, rot=<0 0 0 1>}
 his1_hair_B_b_l {parent=his1_hair_B_a_l, loc=<5.7142 0.0749 0.0694>, rot=<0.0041 0.012 -0.0215 -0.9997>}
 his1_hair_B_a_base_r {parent=his1_hair, loc=<8.3195 -1.6526 6.5389>, rot=<0.0645 -0.975 -0.2123 0.0113>}
 his1_hair_B_a_r {parent=his1_hair_B_a_base_r, loc=<0 0 0>, rot=<0 0 0 1>}
 his1_hair_B_b_r {parent=his1_hair_B_a_r, loc=<5.6899 -0.3815 -0.3052>, rot=<0 0.0526 0 -0.9986>}
 his1_hair_A_a_base_l {parent=his1_hair, loc=<5.2878 3.4082 -7.2631>, rot=<0.0902 -0.9959 -0.0023 0.0024>}
 his1_hair_A_a_l {parent=his1_hair_A_a_base_l, loc=<0 0 0>, rot=<0 0 0 1>}
 his1_hair_A_b_l {parent=his1_hair_A_a_l, loc=<5.1459 -0.1222 -0.2915>, rot=<-0.0111 -0.1362 -0.0392 -0.9898>}
 his1_hair_A_c_l {parent=his1_hair_A_b_l, loc=<3.95 0.1075 0.3923>, rot=<0 0 -0.0325 -0.9995>}
 his1_hair_A_a_base_r {parent=his1_hair, loc=<5.0041 4.9691 6.9574>, rot=<0.9948 0.0899 -0.0478 0.0069>}
 his1_hair_A_a_r {parent=his1_hair_A_a_base_r, loc=<0 0 0>, rot=<0 0 0 1>}
 his1_hair_A_b_r {parent=his1_hair_A_a_r, loc=<-5.214 -0.21 -0.0518>, rot=<-0.0111 -0.1362 -0.0392 -0.9898>}
 his1_hair_A_c_r {parent=his1_hair_A_b_r, loc=<-3.9505 0.2298 -0.321>, rot=<0 0 -0.0325 -0.9995>}
 clavicle_r {parent=spine_05, loc=<5.4831 -0.1053 3.2946>, rot=<0.7289 0.0465 -0.6829 0.009>}
 upperarm_r {parent=clavicle_r, loc=<-9.8658 1.2136 0.4105>, rot=<0.0164 -0.3245 -0.0269 -0.9454>}
]
