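# Physics asset for the face accessory mesh of the F_MED_Impulse_Spring HIS_02 player skin, set up for hair physics.
# Capsules on head, neck, upper spine and upper arms form body shapes around left and right A and B hair chains and a back chain.
# Chain bases join the hair root, which joins the head, through free angular constraints like the body's.
# Segments below each base use tighter per-joint swing and twist limits.
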PhysicsAsset /Game/Characters/Player/Female/Medium/Bodies/F_MED_Impulse_Spring/Skins/HIS_02/Meshes/F_MED_Impulse_Spring_HIS_2_FaceAcc_Physics
bodies {
 head [{shapeType=Capsule, length=4, radius=6, center=<1.1782 0.9924 0>, rotation=<rot 90 0 0>}]
 his1_hair [{shapeType=Capsule, length=1, radius=1}]
 his1_hair_A_a_base_l [{shapeType=Capsule, length=0.5, radius=0.5}]
 his1_hair_A_a_base_r [{shapeType=Capsule, length=0.5, radius=0.5}]
 his1_hair_A_a_l [{shapeType=Capsule, length=2.5, radius=1.75, center=<3 0 0>, rotation=<rot 90 0 0>}]
 his1_hair_A_a_r [{shapeType=Capsule, length=2.25, radius=2, center=<-3 0 0>, rotation=<rot 90 0 0>}]
 his1_hair_A_b_l [{shapeType=Capsule, length=2, radius=2, center=<1.3 0 0>, rotation=<rot 90 0 0>}]
 his1_hair_A_b_r [{shapeType=Capsule, length=2.25, radius=2, center=<-3 0 0>, rotation=<rot 90 0 0>}]
 his1_hair_A_c_l [{shapeType=Capsule, length=2.5, radius=1.75, center=<1.2076 0 -1.0895>, rotation=<rot 90 0 0>}]
 his1_hair_A_c_r [{shapeType=Capsule, length=1.25, radius=2.5, center=<-2.9878 0 1.2564>, rotation=<rot 90 0 0>}]
 his1_hair_B_a_base_l [{shapeType=Capsule, length=0.5, radius=0.5}]
 his1_hair_B_a_base_r [{shapeType=Capsule, length=0.5, radius=0.5}]
 his1_hair_B_a_l [{shapeType=Capsule, length=1.25, radius=3, center=<3 0 0>, rotation=<rot 90 0 0>}]
 his1_hair_B_a_r [{shapeType=Capsule, length=1.75, radius=2.5, center=<3 0 0>, rotation=<rot 90 0 0>}]
 his1_hair_B_b_l [{shapeType=Capsule, length=1.25, radius=3, center=<3 0 0>, rotation=<rot 90 0 0>}]
 his1_hair_B_b_r [{shapeType=Capsule, length=4.5, radius=2.5, center=<5 0 0>, rotation=<rot 90 0 0>}]
 his1_hair_bk_a [{shapeType=Capsule, length=0, radius=4.75, center=<3 0 0>, rotation=<rot 90 0 0>}]
 his1_hair_bk_a_base_l [{shapeType=Capsule, length=0.5, radius=0.5}]
 his1_hair_bk_b [{shapeType=Capsule, length=0, radius=4.75, center=<3 0 0>, rotation=<rot 90 0 0>}]
 neck_01 [{shapeType=Capsule, length=2.75, radius=4, center=<2.2564 0.5078 0>, rotation=<rot 90 0 0>}]
 spine_05 [{
   shapeType=Capsule
   length=6.5
   radius=17.755
   shapeName=spine_05_capsule
   center=<-3.9605 -2.4188 -0.2076>
   rotation=<rot 0.3281 -65.3692 91.249>
  }]
 upperarm_l [{shapeType=Capsule, length=15.5, radius=4, center=<9.1076 0 0>, rotation=<rot 90 0 0>}]
 upperarm_r [{shapeType=Capsule, length=13, radius=6.5, center=<-9.1076 3.256 -2.5198>, rotation=<rot 90 0 180>}]
}
constraints [
 {
  bone1Name=his1_hair
  bone2Name=head
  swing1LimitDegrees=45
  swing1Motion=Free
  swing2LimitDegrees=45
  swing2Motion=Free
  twistLimitDegrees=45
  twistMotion=Free
 }
 {
  bone1Name=head
  bone2Name=neck_01
  swing1LimitDegrees=45
  swing1Motion=Free
  swing2LimitDegrees=45
  swing2Motion=Free
  twistLimitDegrees=45
  twistMotion=Free
 }
 {
  bone1Name=neck_01
  bone2Name=spine_05
  swing1LimitDegrees=45
  swing1Motion=Free
  swing2LimitDegrees=45
  swing2Motion=Free
  twistLimitDegrees=45
  twistMotion=Free
 }
 {
  bone1Name=his1_hair_bk_a_base_l
  bone2Name=his1_hair
  swing1LimitDegrees=45
  swing1Motion=Free
  swing2LimitDegrees=45
  swing2Motion=Free
  twistLimitDegrees=45
  twistMotion=Free
 }
 {
  bone1Name=his1_hair_B_a_base_l
  bone2Name=his1_hair
  swing1LimitDegrees=45
  swing1Motion=Free
  swing2LimitDegrees=45
  swing2Motion=Free
  twistLimitDegrees=45
  twistMotion=Free
 }
 {
  bone1Name=his1_hair_A_a_base_r
  bone2Name=his1_hair
  swing1LimitDegrees=45
  swing1Motion=Free
  swing2LimitDegrees=45
  swing2Motion=Free
  twistLimitDegrees=45
  twistMotion=Free
 }
 {
  bone1Name=his1_hair_B_a_base_r
  bone2Name=his1_hair
  swing1LimitDegrees=45
  swing1Motion=Free
  swing2LimitDegrees=45
  swing2Motion=Free
  twistLimitDegrees=45
  twistMotion=Free
 }
 {
  bone1Name=his1_hair_A_a_base_l
  bone2Name=his1_hair
  swing1LimitDegrees=45
  swing1Motion=Free
  swing2LimitDegrees=45
  swing2Motion=Free
  twistLimitDegrees=45
  twistMotion=Free
 }
 {bone1Name=his1_hair_bk_a, bone2Name=his1_hair_bk_a_base_l, swing1LimitDegrees=10, swing2LimitDegrees=10, twistLimitDegrees=10}
 {bone1Name=his1_hair_bk_b, bone2Name=his1_hair_bk_a, swing1LimitDegrees=30, swing2LimitDegrees=30, twistLimitDegrees=20}
 {bone1Name=his1_hair_B_a_l, bone2Name=his1_hair_B_a_base_l, swing1LimitDegrees=10, swing2LimitDegrees=10, twistLimitDegrees=10}
 {bone1Name=his1_hair_B_b_l, bone2Name=his1_hair_B_a_l, swing1LimitDegrees=30, swing2LimitDegrees=30, twistLimitDegrees=20}
 {bone1Name=his1_hair_A_a_r, bone2Name=his1_hair_A_a_base_r, swing1LimitDegrees=15, swing2LimitDegrees=15, twistLimitDegrees=15}
 {bone1Name=his1_hair_A_b_r, bone2Name=his1_hair_A_a_r, swing1LimitDegrees=10, swing2LimitDegrees=10, twistLimitDegrees=1}
 {bone1Name=his1_hair_B_a_r, bone2Name=his1_hair_B_a_base_r, swing1LimitDegrees=10, swing2LimitDegrees=10, twistLimitDegrees=10}
 {bone1Name=his1_hair_B_b_r, bone2Name=his1_hair_B_a_r, swing1LimitDegrees=30, swing2LimitDegrees=30, twistLimitDegrees=20}
 {bone1Name=his1_hair_A_a_l, bone2Name=his1_hair_A_a_base_l, swing1LimitDegrees=15, swing2LimitDegrees=15, twistLimitDegrees=5}
 {bone1Name=his1_hair_A_b_l, bone2Name=his1_hair_A_a_l, swing1LimitDegrees=15, swing2LimitDegrees=15, twistLimitDegrees=5}
 {bone1Name=his1_hair_A_c_r, bone2Name=his1_hair_A_b_r, swing1LimitDegrees=10, swing2LimitDegrees=10, twistLimitDegrees=1}
 {bone1Name=his1_hair_A_c_l, bone2Name=his1_hair_A_b_l, swing1LimitDegrees=15, swing2LimitDegrees=15, twistLimitDegrees=15}
 {
  bone1Name=upperarm_l
  bone2Name=spine_05
  swing1LimitDegrees=45
  swing1Motion=Free
  swing2LimitDegrees=45
  swing2Motion=Free
  twistLimitDegrees=45
  twistMotion=Free
 }
 {
  bone1Name=upperarm_r
  bone2Name=spine_05
  swing1LimitDegrees=45
  swing1Motion=Free
  swing2LimitDegrees=45
  swing2Motion=Free
  twistLimitDegrees=45
  twistMotion=Free
 }
]
